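; Interactable beacon prop that fires a team-assigned pulse indicator for the interacting player, then goes on cooldown.
; Activation starts pulse markers for one tick and sends a replicated GameplayCue scaled by visual radius.
; An authority timer recomputes remaining seconds from server time, driving prompt text plus material and sound updates through OnRep.
; Interaction stays allowed during cooldown so the prompt shows, with delayed failure feedback handled in Blueprint.
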(event Building|EventOnBuildingActorInitialized
  (params (InitializationReason :type EFortBuildingInitializationReason%20Enum :default "StaticallyPlaced") (BuildingPersistentState :type EFortBuildingPersistentState%20Enum :default "Default"))
  ;[1 1 1] On BGA Initialized
  (|CallParentFunction InitializationReason BuildingPersistentState)
  (Variables|Utility|SetCooldownTimeRemaining 0))

(event EventOnDeathPlayEffects
  (params (Damage :type Float%20%28single-precision%29) (DamageTags :type Gameplay%20Tag%20Container%20Structure%20%28by%20ref%29) (Momentum :type Vector :default "0, 0, 0") (HitInfo :type Hit%20Result%20Structure%20%28by%20ref%29) (InstigatedBy :type Fort%20Pawn%20Object%20Reference) (DamageCauser :type Actor%20Object%20Reference) (EffectContext :type Gameplay%20Effect%20Context%20Handle%20Structure))
  (|CallParentFunction Damage DamageTags Momentum HitInfo InstigatedBy DamageCauser EffectContext)
  (bind _ambientloopaudio (Variables|Default|GetAmbientLoopAudio)
    (:then
      (Audio|Components|Audio|FadeOut _ambientloopaudio 0.5 0.0 "Linear" :!class /Script/Engine.AudioComponent))
    (:else)))

(event Interaction|EventBlueprintOnInteract
  (params (InteractingPawn :type Fort%20Pawn%20Object%20Reference) (InteractionBeingAttempted :type EInteractionBeingAttempted%20Enum :default "FirstInteraction"))
  ;[1 1 1] On Interact(Server)
  (if (Variables|Utility|GetInteractReady)
    (Variables|Utility|SetInteractReady false)
    (CallFunction|ActivatePulseIndicators InteractingPawn)
    (CallFunction|PlayPulseFX InteractingPawn)
    (Utilities|FlowControl|Delay (ScalableFloat|GetValueatLevel (Variables|Balance|GetRow_Cooldown)))
    (Variables|Utility|SetInteractReady true)))

(event Custom|ActivatePulseIndicators
  (params (Pawn :type Fort%20Pawn%20Object%20Reference))
  ;[1 1 1] Activate Pulse Indicators
  (Class|PulseIndicatorComponent|SetDisplayTeam (Variables|Default|GetPulseIndicator) (Team|GetActorTeam Pawn))
  (bind _asfort_player_controller_athena (Utilities|Casting|CastToFortPlayerControllerAthena (Fortnite|GetFortPlayerControllerfromActor Pawn))
    (:then
      (Class|PulseIndicatorComponent|SetIndicatingController (Variables|Default|GetPulseIndicator) _asfort_player_controller_athena))
    (:CastFailed
      _))
  (Class|PulseIndicatorComponent|StartPulseMarkers (Variables|Default|GetPulseIndicator))
  (Utilities|FlowControl|DelayUntilNextTick)
  (Class|PulseIndicatorComponent|StopPulseMarkers (Variables|Default|GetPulseIndicator)))

(event Custom|PlayPulseFX
  (params (Pawn :type Fort%20Pawn%20Object%20Reference))
  ;[1 1 1] Play Pulse FX
  (Ability|ExecuteGameplayCueWithParamsReplicated (Variables|FX|GetTag_GCPulseFX) (Ability|GameplayCue|MakeGameplayCueParameters :NormalizedMagnitude 0.0 :RawMagnitude (ScalableFloat|GetValueatLevel (Variables|Default|GetRow_VisualRadius)) :EffectContext (Class|AthenaFunctionLibrary|MakeSimpleFortEffectContext Pawn Pawn) :Location "0,0,0" :Normal "0,0,0" :Instigator Pawn :EffectCauser Pawn :SourceObject (select (Utilities|IsValid (Variables|Default|GetBeaconPingSourceObject) :!class /Script/Engine.KismetSystemLibrary) (Variables|Default|GetBeaconPingSourceObject) Pawn) :GameplayEffectLevel 1 :AbilityLevel 1 :bReplicateLocationWhenUsingMinimalRepProxy false))
  (CallFunction|PlaySoundfromLocation (Variables|Sound|GetActivateSound))
  ;[1 1 1] Cooldown Text
  (Networking|SwitchHasAuthority
    (:Authority
      (Variables|Utility|SetActivationTime (Fortnite|Sync|Time|GetServerWorldTimeSeconds))
      (bind _returnvalue (Utilities|Time|SetTimerbyEvent (delegate Custom|SetText) 1.0 true))
      (Variables|Utility|SetTimer_Countdown _returnvalue)
      (Variables|Utility|SetCooldownTimeRemaining (Math|Float|Truncate (ScalableFloat|GetValueatLevel (Variables|Balance|GetRow_Cooldown)))))))

(event Custom|SetText
  ;[1 1 1] Play Pulse FX
  ;[1 1 1] Cooldown Text
  (Variables|Utility|SetCooldownTimeRemaining (Math|Integer|Clamp(Integer) (Math|Float|Round (- (+ (Variables|Utility|GetActivationTime) (ScalableFloat|GetValueatLevel (Variables|Balance|GetRow_Cooldown))) (Fortnite|Sync|Time|GetServerWorldTimeSeconds))) 0 (Math|Float|Truncate (ScalableFloat|GetValueatLevel (Variables|Balance|GetRow_Cooldown))))))

(event Interaction|EventBlueprintOnBeginInteract
  (params (InteractingPawn :type Fort%20Pawn%20Object%20Reference) (InteractionBeingAttempted :type EInteractionBeingAttempted%20Enum :default "FirstInteraction") (InteractionType :type TInteractionType%20Enum :default "IT_NoInteraction"))
  ;[1 1 1] Local on Begin Interact
  (if (Variables|Utility|GetInteractReady)
    (|CallParentFunction InteractingPawn InteractionBeingAttempted InteractionType)
    (else
      (Utilities|FlowControl|Delay (+ (Variables|Interact|GetInteractTimeOnCooldown) (Variables|Interact|GetDelay_InteractToPlayFailedInteractAnimation))) ; Add slight delay to allow interact to complete so that failed interact animation plays
      (if (> (- (Fortnite|Sync|Time|GetServerWorldTimeSeconds) (Variables|Utility|GetActivationTime)) (Variables|Interact|GetInteractRetryCooldown))
        (InteractInterface|LocalonFailedInteract (Utilities|Casting|CastToFortPlayerPawn InteractingPawn))
        (CallFunction|PlaySoundfromLocation (Variables|Sound|GetErrorSound))))))

(fn BlueprintCanInteract
  (params (InteractingPawn :type Fort%20Pawn%20Object%20Reference) (InteractionBeingAttempted :type EInteractionBeingAttempted%20Enum :default "FirstInteraction") (InteractionType :type TInteractionType%20Enum :default "IT_NoInteraction"))
  ;[1 1 1] Always allow interact attempts - handle failures in blueprint (setting this false hides the interact text prompt)
  (return true))

(fn BlueprintGetInteractionString
  (params (InteractingPawn :type Fort%20Pawn%20Object%20Reference) (InteractionBeingAttempted :type EInteractionBeingAttempted%20Enum :default "FirstInteraction"))
  (if (Variables|Utility|GetInteractReady)
    (return (Variables|Interact|Text|GetText_Interact))
    (else
      (return (Utilities|Text|FormatText "NSLOCTEXT(\"[F612B408466AEE1B30C9C0AC7086C5F0]\", \"844D393C483A71FF3F2DF780E86D695D\", \"{CooldownText} ({CooldownTimeRemaining})\")" (Variables|Interact|Text|GetText_Cooldown) (CallFunction|GetCooldownTimeRemainingFormatted))))))

(fn BlueprintGetInteractionTime
  (params (InteractingPawn :type Fort%20Pawn%20Object%20Reference) (InteractionBeingAttempted :type EInteractionBeingAttempted%20Enum :default "FirstInteraction"))
  (|CallParentFunction InteractingPawn InteractionBeingAttempted)
  (if (Variables|Utility|GetInteractReady)
    (return 0.0 false)
    (else
      (return (Variables|Interact|GetInteractTimeOnCooldown) true))))

(fn GetCooldownTimeRemainingFormatted
  (LocalVariables|SetMinutes (/ (Variables|Utility|GetCooldownTimeRemaining) 60))
  (LocalVariables|SetSeconds (Math|Integer|%25%28Integer%29 (Variables|Utility|GetCooldownTimeRemaining) 60))
  (if (<= (LocalVariables|GetMinutes) 0)
    (return (Utilities|Text|ToText(Integer) (LocalVariables|GetSeconds)))
    (elif (>= (LocalVariables|GetSeconds) 10)
      (return (Utilities|Text|FormatText "NSLOCTEXT(\"[F612B408466AEE1B30C9C0AC7086C5F0]\", \"D991F63C4B528E79B29A319E1923FBD7\", \"{Minutes}:{Seconds}\")" (LocalVariables|GetMinutes) (LocalVariables|GetSeconds)))
      (else
        (return (Utilities|Text|FormatText "NSLOCTEXT(\"[F612B408466AEE1B30C9C0AC7086C5F0]\", \"A0B7133146C96CDEFE9CE1A186904ACC\", \"{Minutes}:0{Seconds}\")" (LocalVariables|GetMinutes) (LocalVariables|GetSeconds)))))))

(fn LocalOnFailedInteract
  (params (InteractingPawn :type Fort%20Player%20Pawn%20Object%20Reference))
  (|CallParentFunction InteractingPawn)
  (FailedAction|UnabletoPerformAction InteractingPawn true (Variables|Interact|GetTag_FailedInteractReason)))

(fn OnRep_CooldownTimeRemaining
  (if (> (Variables|Utility|GetCooldownTimeRemaining) 0)
    (Rendering|Material|SetScalarParameterValueonMaterials (Variables|Default|GetStaticMesh) (Variables|Default|GetMaterialParameter_CooldownTimeRemaining) (Math|Conversions|ToFloat(Integer) (Variables|Utility|GetCooldownTimeRemaining)))
    (CallFunction|PlaySoundfromLocation (Variables|Sound|GetTimerSound))
    (bind _beaconpingsourceobject (Variables|Default|GetBeaconPingSourceObject)
      (:then
        (if (== (Variables|Utility|GetCooldownTimeRemaining) 5)
          _))
      (:else))
    (else
      (Rendering|Material|SetScalarParameterValueonMaterials (Variables|Default|GetStaticMesh) (Variables|Default|GetMaterialParameter_CooldownTimeRemaining) 0.0)
      (CallFunction|PlaySoundfromLocation (Variables|Sound|GetTimerCompleteSound))))
  (Utilities|FlowControl|DoOnce false
    (:Completed
      (Audio|PlaySoundatLocation (Variables|Default|GetPreCompleteSound) (Transformation|GetActorLocation _beaconpingsourceobject)))))

(fn PlaySoundFromLocation
  (params (SoundToPlay :type Sound%20Base%20Object%20Reference))
  (Audio|PlaySoundatLocation SoundToPlay (Transformation|GetWorldLocation (Variables|Default|GetAmbientLoopAudio))))

(fn ConstructionScript)
