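; BP_Tinfoil_Desk_A is a desk prop Blueprint in the Apollo Tinfoil building actor set, with an S_Agency_Chair_01 chair component.
; Its construction script calls the parent, then adds a small random yaw offset to the chair's current relative rotation, keeping roll and pitch, so the chair's angle varies.
; BeginPlay, Tick and ActorBeginOverlap events are empty.
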(event EventBeginPlay)

(event Collision|EventActorBeginOverlap
  (params (OtherActor :type Actor%20Object%20Reference)))

(event EventTick
  (params (DeltaSeconds :type Float%20%28single-precision%29)))

(fn ConstructionScript
  (|CallParentFunction)
  (Transformation|SetRelativeRotation (Variables|Default|GetS_Agency_Chair_01) (Math|Rotator|MakeRotator (.roll (Variables|/Script/Engine.SceneComponent|GetRelativeRotation (Variables|Default|GetS_Agency_Chair_01))) (.pitch (Variables|/Script/Engine.SceneComponent|GetRelativeRotation (Variables|Default|GetS_Agency_Chair_01))) (+ (.yaw (Variables|/Script/Engine.SceneComponent|GetRelativeRotation (Variables|Default|GetS_Agency_Chair_01))) (Math|Random|RandomFloatinRange -1e+01 1e+01))) false false))
